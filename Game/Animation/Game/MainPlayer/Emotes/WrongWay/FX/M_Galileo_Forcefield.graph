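# Additive unlit hologram forcefield material for Niagara mesh particles in the WrongWay (Galileo) emote FX.
# Two panned, noise-distorted dust layers combine with a vertical pulse, driven by the Wave dynamic parameter, and a glow near the player's position.
# Fresnel and UV gradient sample a curve-atlas color.
# Vertex colors mask regions and select between angle and depth fades, while shading-path switches give mobile constant Fresnel and fade values.
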Material /Game/Animation/Game/MainPlayer/Emotes/WrongWay/FX/M_Galileo_Forcefield
graph {
 properties {
  bUsedWithNiagaraMeshParticles=true
  blendMode=BLEND_Additive
  pixelDepthOffsetMode=PDOM_Legacy
  shadingModel=MSM_Unlit
  usedShadingModels=Unlit
 }
 nodes [
  Constant() {r=1}
  Constant() {r=1}
  n1 = Multiply(A=ScalarParameter() {defaultValue=4, parameterName=Speed}, B=Time())
  n2 = TextureCoordinate()
  n3 = VectorParameter() {parameterName=Tile, defaultValue=<rgba 100 100 0 0>}
  n4 = WorldPosition()
  n5 = AppendVector(A=Multiply(A=ComponentMask(None=@n2) {r=true}, B=@n3.R), B=Divide(A=ComponentMask(None=@n4.XYZ) {b=true}, B=@n3.G))
  n6 = Panner(Coordinate=@n5, Time=@n1)
  n7 = MaterialFunctionCall(Float3=Add(
    A=@n1
    B=TextureSample(UVs=@n6) {samplerType=SAMPLERTYPE_LinearColor, texture=&T_NoisePack_LinearColor}.RGB
   )) {materialFunction=&BreakOutFloat3Components}
  n8 = Sine(None=@n7.R) {period=1.25}
  n9 = Sine(None=@n7.G) {period=1.08}
  n10 = Multiply(A=ConstantBiasScale(None=AppendVector(A=@n8, B=@n9)), B=ScalarParameter() {defaultValue=0.1, parameterName=" NoiseDistortion"}) {constB=0.07}
  n11 = TextureSample(UVs=Add(A=Panner(Coordinate=@n5, Time=@n1) {speedX=0.04, speedY=-0.1}, B=@n10)) {texture=/Game/Global/Textures/Default/PointSplatter/T_Tiling_Dust}
  n12 = Multiply(A=@n5) {constB=0.5}
  n13 = TextureSample(UVs=Subtract(A=Panner(Coordinate=@n12, Time=@n1) {speedX=-0.06, speedY=-0.1}, B=@n10)) {texture=&T_Tiling_Dust}
  n14 = DynamicParameter() {defaultValue=<rgba 0.6 0 -1 1>, paramNames [NoisePower,Opacity,Wave,Param4]}
  n15 = LinearInterpolate(A=LinearInterpolate(A=@n11.RGB, B=@n13.RGB), B=Multiply(A=@n11.RGB, B=@n13.RGB), Alpha=@n14.NoisePower)
  n16 = OneMinus(None=@n14.Wave)
  n17 = ScalarParameter() {defaultValue=1500, parameterName=PulseHeight}
  n18 = ActorPositionWS()
  n19 = ScalarParameter() {defaultValue=400, parameterName=PulseRadius}
  n20 = SphereMask(
   A=ComponentMask(None=WorldPosition().XYZ) {b=true}
   B=Add(A=Multiply(A=@n17, B=@n14.Wave) {constB=1200}, B=ComponentMask(None=@n18) {b=true})
   Radius=LinearInterpolate(A=@n19, B=Multiply(A=@n19) {constB=0.1}, Alpha=@n14.Wave) {constA=500, constB=100}
  ) {hardnessPercent=0}
  n21 = Constant3Vector() {constant=<rgba 0 0 30 0>}
  n22 = CollectionParameter() {collection=&FortniteMaterialParameters, parameterName=PlayerPosition}
  n23 = ComponentMask(None=@n22) {b=true, g=true, r=true}
  n24 = Divide(A=Distance(A=Add(A=@n21, B=@n23), B=WorldPosition().XYZ), B=ScalarParameter() {defaultValue=200, parameterName=FadeInDist})
  n25 = Power(Base=OneMinus(None=Saturate(None=@n24)))
  n26 = Add(A=Power(Base=Multiply(A=@n16, B=@n20)) {constExponent=4}, B=@n25)
  n27 = Clamp(None=Subtract(A=@n14.NoisePower, B=@n26)) {clampMode=CMODE_ClampMin}
  n28 = Multiply(A=Saturate(None=Power(Base=@n15, Exp=@n27)), B=Multiply(A=Add(B=@n26) {constA=1}) {constB=5}) {constB=5}
  n29 = VertexColor()
  n30 = OneMinus(None=@n29.B)
  n31 = Add(A=ScalarParameter() {defaultValue=0.2, parameterName=FresnelBaseBrightness}, B=Multiply(A=Multiply(A=@n30, B=@n25)) {constB=0.5})
  n32 = ShadingPathSwitch(
   Default=Fresnel(ExponentIn=ScalarParameter() {defaultValue=2, parameterName=FresnelExpBase}, BaseReflectFractionIn=Saturate(None=@n31)) {baseReflectFraction=0.2, exponent=1.5}
   Mobile=Constant() {r=0.6}
  ) {desc=Migrated from FeatureLevelSwitch}
  n33 = ComponentMask(None=TextureCoordinate()) {g=true}
  n34 = Power(Base=@n33, Exp=ScalarParameter() {defaultValue=10, parameterName=GradientPower}) {constExponent=3}
  n35 = CurveAtlasRowParameter(CurveTime=Saturate(None=Multiply(A=@n32, B=@n34))) {
   atlas=&AthenaCurveAtlas01
   curve=/Game/Athena/LinearColorCurves/Curve_Galileo_Hologram
   defaultValue=171
   parameterName=Curve
  }
  n36 = ParticleColor()
  n37 = Multiply(A=Add(A=Add(A=@n28, B=@n26), B=@n35), B=@n36.RGB)
  n38 = VertexColor()
  n39 = Multiply(A=ScalarParameter() {defaultValue=8, parameterName=Brightness}, B=@n35)
  n40 = VectorParameter() {parameterName=ColorAdd, defaultValue=<rgba 0.0785 0.2191 0.5 0>}
  MaterialFunctionCall(Emissive=Multiply(A=LinearInterpolate(A=@n37, Alpha=@n38.B), B=Add(A=@n39, B=@n40.RGB))) {materialFunction=&MF_FX_EmissiveAdjustments}
  n41 = Multiply(A=@n32, B=@n33)
  n42 = Constant3Vector() {constant=<rgba 0 0 1 0>}
  n43 = ScalarParameter() {defaultValue=0.5, parameterName=AngleFadeExpBase}
  n44 = Multiply(A=VertexNormalWS(), B=TwoSidedSign())
  n45 = Multiply(
   A=Power(
    Base=MaterialFunctionCall!(Axis=@n42, Exponent=@n43, Normal=@n44) {materialFunction=/Game/Effects/Fort_Effects/Materials/_Functions/MF_AxisAlignedFrensnel}.Fresnel
    Exp=ScalarParameter() {defaultValue=2.5, parameterName=AngleFadePower}
   )
   B=ScalarParameter() {defaultValue=3, parameterName=AngleFadeMult}
  ) {constB=5}
  n46 = Constant() {r=1}
  n47 = MaterialFunctionCall("FadeSensitivity (optional)"=@n46) {materialFunction=&FadeBasedOnViewAngle}
  n48 = Multiply(A=OneMinus(None=@n33), B=ScalarParameter() {defaultValue=100, parameterName=MaxDepthFade})
  n49 = ShadingPathSwitch(Default=LinearInterpolate(A=Saturate(None=@n45), B=DepthFade(Opacity=@n47.Result, FadeDistance=@n48), Alpha=@n29.B), Mobile=@n46) {desc=Migrated from FeatureLevelSwitch}
  n50 = ScalarParameter() {defaultValue=0.5, parameterName=BottomEdgeOpacity}
  n51 = Add(A=@n14.Opacity, B=@n25)
  n52 = VertexColor()
  Multiply(A=@n36.A, B=Multiply(A=Multiply(A=@n41, B=@n49), B=LinearInterpolate(A=@n50, B=@n51, Alpha=@n52.R) {constA=1}))
  Sine(None=@n7.B) {period=1.83}
 ]
}
